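# Niagara system for the StageCue emote's spotlight, with time-of-day colors and a front-end variant for the beam's day color.
# SpotLight draws a two-renderer mesh beam, while Light adds a waveform-flickered light.
# User.NormalizedNotifyProgress indexes their curves, so beam scale, alpha and light intensity follow the notify rather than particle age.
# BallsShadows pins 16 flat dark sprites to skeletal mesh bones by spawn group, culled by a box kill volume.
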NiagaraSystem /BRCosmetics/Animation/Game/MainPlayer/Emotes/StageCue/FX/NS_StageCue_Emote_SpotLight
dependencies {
 usedDataInterfaces [
  &NiagaraDataInterfaceVectorCurve
  &NiagaraDataInterfaceCurve
  &NiagaraDataInterfaceArrayInt32
  &NiagaraDataInterfaceSkeletalMesh
 ]
 usedDynamicInputs []
 usedModules [
  &self:SystemSpawnScript.SystemScriptSource.SystemScriptGraph.NiagaraNodeAssignment_1.SetVariables_#1
  &SystemState
 ]
}
emitters {
 BallsShadows {
  emitterUpdateScript [
   &EmitterState(
    Inactive Response=::NewEnumerator1
    Life Cycle Mode=::NewEnumerator1
    Loop Behavior=::NewEnumerator1
    Loop Duration Mode=::NewEnumerator1
    UseLoopDelay=false
   )
   &SpawnBurst_Instantaneous(Spawn Count=16, Use Spawn Probability=false, UseLoopCountLimit=false)
  ]
  particleSpawnScript [
   &FN_SetSpawnGroup(Number of Spawn Groups=16)
   &InitializeParticle(
    Color=&FNTimeOfDayColor(Night Color=<rgba 0 0 0 0.325>, Morning Color=<rgba 0 0 0 0.325>, Day Color=<rgba 0 0 0 0.4>, Evening Color=<rgba 0 0 0 0.325>)
    Mesh Scale Mode=::NewEnumerator4
    Sprite Size Mode=::NewEnumerator3
    Uniform Sprite Size=7
    UsePositionOffset=false
    Write Material Random=true
   )
   &SpriteFacingAndAlignment(Sprite Alignment=<1 0 0>, Sprite Facing=<0 0 1>, WriteAlignment=true, WriteFacing=true)
  ]
  particleUpdateScript [
   &ParticleState(Kill Particles When Lifetime Has Elapsed=false, Let Infinitely Lived Particles Die When Emitter Deactivates=true)
   &SkeletalMeshLocation(
    Bone / Socket Index=&SelectIntFromArray(
     Direct Array Index=$Particles.SpawnGroup
     Int Selection Array=DataInterface() {propertyValues {IntData [0,1,2,3,4,5,6,7,8,9,10,11,12,13,14,15]}}
    )
    Bone Sampling Mode=::NewEnumerator4
    MaskSpawnGroup=false
    Sampled Position Offset=<0 0 -1.25>
    Sampled Position Offset Coordinate Space:ENiagaraSkelMeshTransforms=::NewEnumerator2
    Skeletal Mesh=$System.Skeletal Mesh
   )
   &KillParticlesInVolume(Box Size:Vector3f=<45 50 25>, Kill Shape=::NewEnumerator1)
  ]
  rendererClasses [&NiagaraSpriteRendererProperties]
  renderers [{rendererClass=&NiagaraSpriteRendererProperties}]
 }
 Light {
  emitterUpdateScript [
   &EmitterState(
    Inactive Response=::NewEnumerator1
    Life Cycle Mode=::NewEnumerator1
    Loop Behavior=::NewEnumerator1
    Loop Duration Mode=::NewEnumerator1
    UseLoopDelay=false
   )
   &SpawnBurst_Instantaneous(Use Spawn Probability=false, UseLoopCountLimit=false)
   /Niagara/Noise/Waveform(
    EnableExponent1=false
    Global Amplitude Scale=0.6
    GlobalClampMax=false
    GlobalClampMin=false
    UseInput1=false
    UsePhase1=false
    "Waveform [1]":ENiagara_Waveforms=::NewEnumerator12
    "[1] Frequency"=4
   )
   /SetVariables(Emitter.Flicker:NiagaraFloat=&Add_Float(A=&Multiply_Float(A=$Emitter.Waveform.Waveform, B=0.5), B=0.5)) {isSetParameters=true}
  ]
  particleSpawnScript [&InitializeParticle(
    Color=&FNTimeOfDayColor(
     Night Color=<rgba 1 0.8521 0.5492 0.7>
     Morning Color=<rgba 1 0.8366 0.5 0.7>
     Day Color=<rgba 1 0.8366 0.5 0.7>
     Evening Color=<rgba 1 0.8366 0.5 0.7>
    )
    Position Offset=<0 -50 50>
    Sprite Size Mode=::NewEnumerator3
    Uniform Sprite Size=120
    UsePositionOffset=true
    Write Material Random=true
   )]
  particleUpdateScript [
   &ParticleState(Kill Particles When Lifetime Has Elapsed=false, Let Infinitely Lived Particles Die When Emitter Deactivates=true)
   &ScaleColor(Scale Alpha=&Multiply_Float(A=$Emitter.Flicker, B=0.3), ScaleA=true, ScaleRGB=true)
   &FN_Light_Attributes(
    Color=$Particles.Color
    Create Color Attribute=true
    Create Exponent Attribute=true
    Create Radius Attribute=true
    Exponent=3
    Radius=65
    Scale Color=&Multiply_Float(
     A=&FloatFromCurve(
      FloatCurve=DataInterface() {propertyValues {Curve {keys [{arriveTangent=238.0882, time=0.03, value=0}, {arriveTangent=51.3522, time=0.12, value=10}, <at 0.4 10>, <at 0.5 1>, <at 1 0>]}}}
      CurveIndex=$User.NormalizedNotifyProgress
     )
     B=$Emitter.Flicker
    )
   )
  ]
  rendererClasses [&NiagaraLightRendererProperties]
  renderers [{rendererClass=&NiagaraLightRendererProperties, rendererIndex=0}]
 }
 SpotLight {
  emitterUpdateScript [
   &EmitterState(
    Inactive Response=::NewEnumerator1
    Life Cycle Mode=::NewEnumerator1
    Loop Behavior=::NewEnumerator1
    Loop Duration Mode=::NewEnumerator1
    UseLoopDelay=false
   )
   &SpawnBurst_Instantaneous(Use Spawn Probability=false, UseLoopCountLimit=false)
  ]
  particleSpawnScript [
   &InitializeParticle(
    Color=&FNTimeOfDayColor(
     Night Color=<rgba 0.5 0.4132 0.25 0.6>
     Morning Color=<rgba 0.5 0.4132 0.25 0.6>
     Day Color=&MakeCustomLinearColorFromBool(
      Bool=&Boolean_OrOperation(Bool1=$User.bIsFrontEnd, Bool2=$User.bIsFrontEndPreview)
      True Color=<rgba 0.8 0.632 0.24 0.275>
      False Color=<rgba 1 0.6877 0.2311 0.5>
     )
     Evening Color=<rgba 0.5 0.4132 0.25 0.6>
    )
    Lifetime Max=0.6
    Lifetime Min=0.8
    Lifetime Mode=::NewEnumerator1
    Mesh Scale=<0.65 0.9 0.6>
    Mesh Scale Mode=::NewEnumerator4
    Position Offset=<0 0 100>
    Sprite Size Mode=::NewEnumerator3
    Uniform Sprite Size=77
    UsePositionOffset=true
    Write Material Random=true
   )
   &InitialMeshOrientation(Mesh Orientation Mode=::NewEnumerator7, Use Rotation Vector=true)
   /SetVariables(Particles.GlowCameraOffset=33) {isSetParameters=true}
  ]
  particleUpdateScript [
   &ParticleState(Kill Particles When Lifetime Has Elapsed=false, Let Infinitely Lived Particles Die When Emitter Deactivates=true)
   &ScaleMeshSize(Scale Factor=&VectorFromCurve(
     VectorCurve=DataInterface() {propertyValues {
       XCurve {keys [<at 0 0>,<at 0.01 0.8>,<at 0.99 0.8>,<at 1 0>]}
       YCurve {keys [<at 0 0>,<at 0.01 0.8>,<at 0.99 0.8>,<at 1 0>]}
       ZCurve {keys [<at 0 1>,<at 1 1>]}
      }}
     CurveIndex=$User.NormalizedNotifyProgress
    )) {module=ScaleMeshSize001}
   &ScaleSpriteSize(
    Uniform Curve Index=$User.NormalizedNotifyProgress
    Uniform Curve Sprite Scale=DataInterface() {propertyValues {Curve {keys [{tangentMode=RCTM_User, time=0, value=0}, {tangentMode=RCTM_User, time=0.03, value=1}]}}}
   ) {module=ScaleSpriteSize001}
   /SetVariables(Particles.GlowColor=&MultiplyLinearColor(Linear Color=$Particles.Color, Scale Factor=&MakeVector4_FromFloat(Float=0.5))) {isSetParameters=true}
   &ScaleColor(
    Scale Alpha=&FloatFromCurve(
     FloatCurve=DataInterface() {propertyValues {Curve {keys [<at 0 0>,<at 0.01 1>,<at 0.99 1>,<at 1 0>]}}}
     CurveIndex=$User.NormalizedNotifyProgress
    )
    ScaleA=true
    ScaleRGB=true
   )
   /SetVariables() {isSetParameters=true}
  ]
  rendererClasses [&NiagaraMeshRendererProperties]
  renderers [
   {rendererClass=&NiagaraMeshRendererProperties, rendererIndex=0}
   {rendererClass=&NiagaraMeshRendererProperties, rendererIndex=1}
  ]
 }
}
summary {
 systemName=NS_StageCue_Emote_SpotLight
 emitters [
  {emitterName=SpotLight, rendererClasses [&NiagaraMeshRendererProperties]}
  {emitterName=Light, rendererClasses [&NiagaraLightRendererProperties]}
  {emitterName=BallsShadows, rendererClasses [&NiagaraSpriteRendererProperties]}
 ]
 userVariables [
  User.bIsFrontEnd {type=NiagaraBool}
  User.bIsFrontEndPreview {type=NiagaraBool}
  User.NormalizedNotifyProgress {type=NiagaraFloat, defaultValue=0, description=""}
 ]
}
